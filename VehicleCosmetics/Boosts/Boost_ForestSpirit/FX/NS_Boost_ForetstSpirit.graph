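# Forest Spirit vehicle boost cosmetic: a Niagara trail placed at the boost position plus car offset.
# Butterfly sprites/meshes swirl via curl noise and point attraction, alongside a mesh cone with animated material parameters.
# One ribbon emitter derives two sine/cosine spiral strands from particle age; TrailSprites merges three spawn groups using render tags.
# Painted-variant modules and user paint parameters recolor the default green-yellow palette, often with randomly picked shades.
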NiagaraSystem /VehicleCosmetics/Boosts/Boost_ForestSpirit/FX/NS_Boost_ForetstSpirit
dependencies {
 usedDataInterfaces [
  &NiagaraDataInterfaceArrayColor
  &NiagaraDataInterfaceCurve
  &NiagaraDataInterfaceVectorCurve
  &NiagaraDataInterfaceArrayFloat
  &NiagaraDataInterfaceSkeletalMesh
 ]
 usedDynamicInputs []
 usedModules [
  &self:SystemSpawnScript.SystemScriptSource.SystemScriptGraph.NiagaraNodeAssignment_0.SetVariables_#1
  &SystemState
 ]
}
emitters {
 Buterflies001 {
  emitterUpdateScript [
   &EmitterState()
   &SpawnRate(SpawnRate=30, Use Spawn Probability=false)
  ]
  particleSpawnScript [
   &VC_PaintedVariantsOverride(
    Black=<rgba 0.2604 0.2604 0.2604 0.3>
    BurntSienna=&SelectLinearColorFromArray(
     Array Sampling Mode=::NewEnumerator0
     Direct Array Index=?
     Color Selection Array=DataInterface() {propertyValues {ColorData [<rgba 2.5 0.5462 0.2829 0.08>, <rgba 0.4219 0.1318 0.0527 0.08>]}}
     Random Seed=0
    )
    Cobalt=&SelectLinearColorFromArray(
     Array Sampling Mode=::NewEnumerator0
     Direct Array Index=?
     Color Selection Array=DataInterface() {propertyValues {ColorData [<rgba 0.9479 1.0599 7 0.25>, <rgba 0.9479 1.0599 7 0.18>]}}
     Random Seed=0
    )
    Crimson=&SelectLinearColorFromArray(
     Array Sampling Mode=::NewEnumerator0
     Direct Array Index=?
     Color Selection Array=DataInterface() {propertyValues {ColorData [<rgba 2 0.0937 0.1422 0.5>, <rgba 0.4627 0 0.0118 0.5>]}}
     Random Seed=0
    )
    DefaultColor=&SelectLinearColorFromArray(
     Array Sampling Mode=::NewEnumerator0
     Direct Array Index=?
     Color Selection Array=DataInterface() {propertyValues {ColorData [<rgba 0.7011 0.7758 0.159 1>, <rgba 0.1683 0.8148 0.0561 1>]}}
     Random Seed=0
    )
    ForestGreen=&SelectLinearColorFromArray(
     Array Sampling Mode=::NewEnumerator0
     Direct Array Index=?
     Color Selection Array=DataInterface() {propertyValues {ColorData [<rgba 0.3125 2 0.3916 0.5>, <rgba 0.1134 0.6406 0.1382 1>]}}
     Random Seed=0
    )
    Grey=&SelectLinearColorFromArray(
     Array Sampling Mode=::NewEnumerator0
     Direct Array Index=?
     Color Selection Array=DataInterface() {propertyValues {ColorData [<rgba 1.2 1.2 1.2 0.5>, <rgba 0.6354 0.6354 0.6354 0.3>]}}
     Random Seed=0
    )
    Lime=&SelectLinearColorFromArray(
     Array Sampling Mode=::NewEnumerator0
     Direct Array Index=?
     Color Selection Array=DataInterface() {propertyValues {ColorData [<rgba 1.556 2.5 0.612 0.5>, <rgba 0.4822 0.7812 0.1831 0.4>]}}
     Random Seed=0
    )
    Orange=&SelectLinearColorFromArray(
     Array Sampling Mode=::NewEnumerator0
     Direct Array Index=?
     Color Selection Array=DataInterface() {propertyValues {ColorData [<rgba 6 0.5842 0 0.25>, <rgba 0.6706 0.2353 0 0.3>]}}
     Random Seed=0
    )
    Pink=&SelectLinearColorFromArray(
     Array Sampling Mode=::NewEnumerator0
     Direct Array Index=?
     Color Selection Array=DataInterface() {propertyValues {ColorData [<rgba 2 1.0729 1.6666 0.6>, <rgba 2 0.6667 1.5205 0.7826>]}}
     Random Seed=0
    )
    Purple=&SelectLinearColorFromArray(
     Array Sampling Mode=::NewEnumerator0
     Direct Array Index=?
     Color Selection Array=DataInterface() {propertyValues {ColorData [<rgba 0.9646 0.3214 3 0.5>, <rgba 0.4823 0.1607 1.5 0.5>]}}
     Random Seed=0
    )
    R1 Black:LinearColor=&SelectLinearColorFromArray(
     Array Sampling Mode=::NewEnumerator0
     Direct Array Index=?
     Color Selection Array=DataInterface() {propertyValues {ColorData [<rgba 0.0365 0.0365 0.0365 1>, <rgba 0.2083 0.2083 0.2083 1>]}}
     Random Seed=0
    )
    R1 BurntSienna:LinearColor=&SelectLinearColorFromArray(
     Array Sampling Mode=::NewEnumerator0
     Direct Array Index=?
     Color Selection Array=DataInterface() {propertyValues {ColorData [<rgba 0.4456 0.1286 0.0418 1>, <rgba 0.0729 0.0228 0.0091 1>]}}
     Random Seed=0
    )
    R1 Cobalt:LinearColor=&SelectLinearColorFromArray(
     Array Sampling Mode=::NewEnumerator0
     Direct Array Index=?
     Color Selection Array=DataInterface() {propertyValues {ColorData [<rgba 0.1823 0.1974 1 1>, <rgba 0.6771 0.7571 5 1>]}}
     Random Seed=0
    )
    R1 Crimson:LinearColor=&SelectLinearColorFromArray(
     Array Sampling Mode=::NewEnumerator0
     Direct Array Index=?
     Color Selection Array=DataInterface() {propertyValues {ColorData [<rgba 1.5 0.0703 0.1067 1>, <rgba 0.4627 0 0.0118 1>]}}
     Random Seed=0
    )
    R1 DefaultColor=&SelectLinearColorFromArray(
     Array Sampling Mode=::NewEnumerator0
     Direct Array Index=?
     Color Selection Array=DataInterface() {propertyValues {ColorData [<rgba 1.8074 2 0.4098 1>, <rgba 0.1683 0.8148 0.0561 1>]}}
     Random Seed=0
    )
    R1 ForestGreen:LinearColor=&SelectLinearColorFromArray(
     Array Sampling Mode=::NewEnumerator0
     Direct Array Index=?
     Color Selection Array=DataInterface() {propertyValues {ColorData [<rgba 0.3628 0.8148 0.2843 1>, <rgba 0.0382 0.7344 0.0409 1>]}}
     Random Seed=0
    )
    R1 Grey:LinearColor=&SelectLinearColorFromArray(
     Array Sampling Mode=::NewEnumerator0
     Direct Array Index=?
     Color Selection Array=DataInterface() {propertyValues {ColorData [<rgba 0.775 0.775 0.775 1>, <rgba 0.2083 0.2083 0.2083 1>]}}
     Random Seed=0
    )
    R1 Lime:LinearColor=&SelectLinearColorFromArray(
     Array Sampling Mode=::NewEnumerator0
     Direct Array Index=?
     Color Selection Array=DataInterface() {propertyValues {ColorData [<rgba 0.9772 1.1 0.0859 1>, <rgba 0.3099 0.5365 0 1>]}}
     Random Seed=0
    )
    R1 Orange:LinearColor=&SelectLinearColorFromArray(
     Array Sampling Mode=::NewEnumerator0
     Direct Array Index=?
     Color Selection Array=DataInterface() {propertyValues {ColorData [<rgba 1.4557 0.5569 0 1>, <rgba 0.75 0.1117 0 1>]}}
     Random Seed=0
    )
    R1 Pink:LinearColor=&SelectLinearColorFromArray(
     Array Sampling Mode=::NewEnumerator0
     Direct Array Index=?
     Color Selection Array=DataInterface() {propertyValues {ColorData [<rgba 2 1.0729 1.6666 1>, <rgba 0.5104 0.1542 0.3823 1>]}}
     Random Seed=0
    )
    R1 Purple:LinearColor=&SelectLinearColorFromArray(
     Array Sampling Mode=::NewEnumerator0
     Direct Array Index=?
     Color Selection Array=DataInterface() {propertyValues {ColorData [<rgba 0.6932 0.1977 1.0141 1>, <rgba 0.1716 0.0005 0.6302 1>]}}
     Random Seed=0
    )
    R1 Saffron:LinearColor=&SelectLinearColorFromArray(
     Array Sampling Mode=::NewEnumerator0
     Direct Array Index=?
     Color Selection Array=DataInterface() {propertyValues {ColorData [<rgba 1 0.83 0.276 1>, <rgba 0.8802 0.6736 0 1>]}}
     Random Seed=0
    )
    R1 SkyBlue:LinearColor=&SelectLinearColorFromArray(
     Array Sampling Mode=::NewEnumerator0
     Direct Array Index=?
     Color Selection Array=DataInterface() {propertyValues {ColorData [<rgba 0.4688 1.8203 2.5 1>, <rgba 0.0908 0.4598 0.6453 1>]}}
     Random Seed=0
    )
    Saffron=&SelectLinearColorFromArray(
     Array Sampling Mode=::NewEnumerator0
     Direct Array Index=?
     Color Selection Array=DataInterface() {propertyValues {ColorData [<rgba 5 4.1501 1.3802 0.3>, <rgba 5 4.1685 1.4583 0.2>]}}
     Random Seed=0
    )
    SkyBlue=&SelectLinearColorFromArray(
     Array Sampling Mode=::NewEnumerator0
     Direct Array Index=?
     Color Selection Array=DataInterface() {propertyValues {ColorData [<rgba 0.5322 2.0668 2.8385 0.5>, <rgba 0.6619 2.3355 3.1771 0.4>]}}
     Random Seed=0
    )
    TitaniumWhite=<rgba 0.8281 0.8281 0.8281 1>
    Use R1 Painted Overrides=true
   )
   &InitializeParticle(
    Color Mode=::NewEnumerator0
    Lifetime=&RandomRangeFloat(Minimum=0.5, Maximum=1)
    Mesh Scale Mode=::NewEnumerator1
    Mesh Uniform Scale Max=0.8
    Mesh Uniform Scale Min=1.2
    Position Offset=&AddVectorToPosition(
     Position=<-100 0 0>
     Vector=&AddVectorToPosition(Position=$User.DTCarOffset, Vector=Linked() {linkedVariable {name=User.SingleBoostPos, type=NiagaraPosition}})
    )
    Position Offset Coordinate Space=::Local
    Sprite Rotation Mode=::NewEnumerator2
    UsePositionOffset=true
    Write Material Random=true
   )
   &AddVelocity(Velocity=&RandomRangeVector(Minimum=<-100 -100 -100>, Maximum=<100 100 100>), Velocity Speed Scale=2)
   &ApplyInitialForces()
   &Color(Color=&MakeCustomLinearColorFromBool(
     Bool=$User.bIsPainted
     True Color=$User.PaintedColor
     False Color=&SelectLinearColorFromArray(
      Array Sampling Mode=::NewEnumerator0
      Direct Array Index=?
      Color Selection Array=DataInterface() {propertyValues {ColorData [<rgba 0.7011 0.7758 0.159 1>, <rgba 0.1683 0.8148 0.0561 1>]}}
      Random Seed=0
     )
    )) {enabled=false}
   &ShapeLocation(
    MaskSpawnGroup=false
    Offset Coordinate Space=::Local
    Offset Mode=::NewEnumerator0
    Override Seed=false
    Sphere Radius=80
    UseOwnerScale=true
   )
  ]
  particleUpdateScript [
   &ParticleState()
   &CurlNoiseForce(Bias Noise=false, Noise Strength=&RandomRangeFloat(Minimum=100, Maximum=200), Pan Noise=false)
   &PointAttractionForce(
    Attraction Radius=500
    Attraction Strength=6
    Attractor Position Offset:Vector3f=&AddVectorToPosition(
     Position=<-50 0 0>
     Vector=&AddVectorToPosition(Position=$User.DTCarOffset, Vector=Linked() {linkedVariable {name=User.SingleBoostPos, type=NiagaraPosition}})
    )
    Kill Within Radius=false
    Module.Use Falloff=false
    UseAttractorOffset=true
   )
   &Drag(Drag=2, Use Linear Drag=true, Use Rotational Drag=true)
   &SolveForcesAndVelocity(Clamp Velocity=false, Limit Acceleration=false)
   &ScaleColor(
    Scale Alpha=&MakeCustomFloatFromBool(Bool=$User.bIsPainted, True Float=0.5, False Float=0.85)
    Scale RGB=&MakeCustomVectorFromBool(
     Bool=$User.bIsPainted
     True Vector=&VectorFromFloat(Value=&SelectFloatFromArray(
       Array Sampling Mode=::NewEnumerator0
       Direct Array Index=?
       Float Selection Array=DataInterface() {propertyValues {FloatData [5,1]}}
       Random Seed=0
      ))
     False Vector=&VectorFromFloat(Value=&FloatFromCurve(FloatCurve=DataInterface() {propertyValues {Curve {keys [<at 0 4>,<at 1 1>]}}}))
    )
    ScaleA=true
    ScaleRGB=true
   )
   &DynamicMaterialParameters(
    Index 0 Param 1=&RandomRangeFloat(Minimum=1, Maximum=2)
    Index 0 Param 2=&RandomRangeFloat(Minimum=0, Maximum=1)
    Module.Index0Param0Write=true
    Module.Index0Param1Write=true
    Module.Index0Param2Write=true
    Module.Index0Param3Write=true
   )
  ]
  rendererClasses [&NiagaraSpriteRendererProperties, &NiagaraMeshRendererProperties]
  renderers [
   {rendererClass=&NiagaraSpriteRendererProperties}
   {rendererClass=&NiagaraMeshRendererProperties, rendererIndex=1}
  ]
 }
 Cone {
  emitterUpdateScript [
   &EmitterState(
    Inactive Response=::NewEnumerator1
    Life Cycle Mode=::NewEnumerator1
    Loop Behavior=::NewEnumerator1
    Loop Duration Mode=::NewEnumerator1
    UseLoopDelay=false
   )
   &SpawnRate(SpawnRate=3, Use Spawn Probability=false)
  ]
  particleSpawnScript [
   &VC_PaintedVariantsOverride(
    Black=<rgba 0.3438 0.3438 0.3438 0.5>
    BurntSienna=<rgba 0.5417 0.257 0.1806 0.5>
    Cobalt=<rgba 0.7248 0.8345 1.42 0.5>
    Crimson=<rgba 1.0938 0.1196 0.1701 0.5>
    DefaultColor=<rgba 0.7347 2.1 0.3828 0.4>
    ForestGreen=<rgba 0.29 1.45 0.3115 0.4>
    Grey=<rgba 0.7552 0.7552 0.7552 0.5>
    Lime=<rgba 0.9302 1 0 0.5>
    Orange=<rgba 1.2522 0.5645 0 0.4>
    Pink=<rgba 1 0.62 0.8618 0.5>
    Purple=<rgba 0.8706 0.1851 1.1849 0.5>
    R1 Black:LinearColor=<rgba 0.1719 0.1719 0.1719 1>
    R1 BurntSienna:LinearColor=<rgba 0.2604 0.1382 0.1056 0.5>
    R1 Cobalt:LinearColor=<rgba 0.36 0.3643 0.9635 0.5>
    R1 Crimson:LinearColor=<rgba 0.6302 0.0624 0.0883 0.5>
    R1 DefaultColor=<rgba 0.3726 1.0391 0.1948 0.5>
    R1 ForestGreen:LinearColor=<rgba 0.2481 0.86 0.129 0.5>
    R1 Grey:LinearColor=<rgba 0.4688 0.4688 0.4688 0.4>
    R1 Lime:LinearColor=<rgba 0.5275 0.6979 0.0243 0.5>
    R1 Orange:LinearColor=<rgba 0.75 0.3086 0.0212 0.4>
    R1 Pink:LinearColor=<rgba 0.8438 0.3643 0.591 0.5>
    R1 Purple:LinearColor=<rgba 0.3179 0 0.625 0.5>
    R1 Saffron:LinearColor=<rgba 0.9115 0.7331 0.1472 0.5>
    R1 SkyBlue:LinearColor=<rgba 0.0938 0.6362 0.7604 0.5>
    R1 TitaniumWhite:LinearColor=<rgba 0.8385 0.8385 0.8385 0.5>
    Saffron=<rgba 1.5 1.2065 0.2422 0.5>
    SkyBlue=<rgba 0 1.4638 2.2 0.5>
    TitaniumWhite=<rgba 1 1 1 0.5>
    Use R1 Painted Overrides=true
   )
   &InitializeParticle(
    Color Mode=::NewEnumerator0
    Lifetime=1.2
    Mesh Scale=&Multiply_VectorByFloat(Vector=<5 5 6>, Float=1.19)
    Mesh Scale Mode=::NewEnumerator4
    Position Offset=&AddVectorToPosition(
     Position=<-30 0 0>
     Vector=&AddVectorToPosition(Position=$User.DTCarOffset, Vector=Linked() {linkedVariable {name=User.SingleBoostPos, type=NiagaraPosition}})
    )
    Position Offset Coordinate Space=::Local
    Sprite Rotation Mode=::NewEnumerator2
    UsePositionOffset=true
    Write Material Random=true
   )
   &AddVelocity(Velocity=<-1500 0 0>) {enabled=false}
   &InitialMeshOrientation(Mesh Orientation Mode=::NewEnumerator7, Rotation=<0 -1.25 0>, Use Rotation Vector=true)
   &Color(Color=&MakeCustomLinearColorFromBool(Bool=$User.bIsPainted, True Color=$User.PaintedColor, False Color=<rgba 0.1 2 0.05 1>)) {enabled=false}
  ]
  particleUpdateScript [
   &ParticleState()
   &ScaleMeshSize(Scale Factor=&VectorFromFloat(Value=&FloatFromCurve(FloatCurve=DataInterface() {propertyValues {Curve {keys [{tangentMode=RCTM_User, time=0, value=0.3}, {tangentMode=RCTM_User, time=1, value=1}]}}}))) {enabled=false}
   &SolveForcesAndVelocity(Clamp Velocity=false, Limit Acceleration=false)
   &DynamicMaterialParameters(
    Index 0 Param 1=&FloatFromCurve(FloatCurve=DataInterface() {propertyValues {Curve {keys [{tangentMode=RCTM_User, time=0, value=-1}, {tangentMode=RCTM_User, time=1, value=2}]}}})
    Index 0 Param 2=0.3
    Module.Index0Param0Write=true
    Module.Index0Param1Write=true
    Module.Index0Param2Write=false
    Module.Index0Param3Write=false
   )
   &ScaleColor(
    Scale RGB=&VectorFromFloat(Value=&MakeCustomFloatFromBool(Bool=true, True Float=2, False Float=1))
    ScaleA=true
    ScaleRGB=true
   ) {enabled=false}
  ]
  rendererClasses [&NiagaraSpriteRendererProperties, &NiagaraMeshRendererProperties]
  renderers [
   {rendererClass=&NiagaraSpriteRendererProperties}
   {rendererClass=&NiagaraMeshRendererProperties, rendererIndex=1}
   {rendererClass=&NiagaraMeshRendererProperties, rendererIndex=2}
  ]
 }
 Ribbon002 {
  emitterUpdateScript [
   &EmitterState()
   &VC_FilteredSocketCount()
   &VC_RateQualityHandler(Epic Spawn Rate=100, High Spawn Rate=80, Low Spawn Rate=80, Medium Spawn Rate=80, Use Custom Rate?=true)
   &VC_SpawnRibbon_v1(Spawn Time=-1, SpawnRate=$Emitter.Spawn Rate)
  ]
  particleSpawnScript [
   &VC_PaintedVariantsOverride(
    Black=<rgba 0.0312 0.0312 0.0312 1>
    Cobalt=<rgba 0.4905 0.3962 1.901 0.2>
    Crimson=<rgba 0.349 0 0.0095 1>
    DefaultColor=<rgba 0.2 0.5 0.25 0.5>
    Lime=<rgba 0.2943 0.5885 0 0.8>
    Orange=<rgba 0.4062 0.0736 0 1>
    Pink=<rgba 0.5365 0.1229 0.3861 1>
    Purple=<rgba 0.1514 0 0.4583 0.9478>
    Saffron=<rgba 0.5833 0.4472 0 1>
    TitaniumWhite=<rgba 0.6771 0.6771 0.6771 1>
   )
   &VC_InitializeRibbon_v1(
    Color=$StackContext.Color
    Custom Offset=true
    Lifetime=0.8
    Local Vector Offset=&AddVectorToPosition(
     Position=<-28 0 0>
     Vector=&AddVectorToPosition(Position=$User.DTCarOffset, Vector=Linked() {linkedVariable {name=User.SingleBoostPos, type=NiagaraPosition}})
    )
    Offset Amount=1
    Ribbon Width=80
   )
   &AddVelocity(Velocity=<-1800 0 0>) {module=AddVelocity001}
   &Color(Color=&MakeCustomLinearColorFromBool(Bool=$User.bIsPainted, True Color=$User.PaintedColor, False Color=<rgba 0.2 0.5 0.25 0.5>)) {enabled=false}
   /SetVariables(
    Particles.ColorTag1:LinearColor=&MakeCustomLinearColorFromBool(Bool=$User.bIsPainted, True Color=$User.PaintedColor, False Color=<rgba 0 0.6406 0.0078 1>)
    Particles.ColorTag2:LinearColor=&MakeCustomLinearColorFromBool(Bool=$User.bIsPainted, True Color=$User.PaintedColor2, False Color=<rgba 0.7011 0.7758 0.159 1>)
    Particles.WidthTag1=35
    Particles.WidthTag2=10
   ) {isSetParameters=true}
  ]
  particleUpdateScript [
   &ParticleState()
   &InheritSourceMovement(
    Applied Position Scale=&VectorFromFloat(Value=&FloatFromCurve(FloatCurve=DataInterface() {propertyValues {Curve {keys [<at 0 1 RCIM_Cubic>,<at 1 0 RCIM_Cubic>]}}}))
    ApplyForce=false
    ApplyPosition=true
    ApplyPositionFalloff=false
   )
   &SolveForcesAndVelocity(Clamp Velocity=false, Limit Acceleration=false)
   &VC_UpdateRibbon_v1()
   &ScaleColor(
    Scale Alpha=&MakeCustomFloatFromBool(
     Bool=$User.bIsPainted
     True Float=&FloatFromCurve(FloatCurve=DataInterface() {propertyValues {Curve {keys [<at 0 1>,<at 1 0>]}}}, Scale Curve=0.15)
     False Float=&FloatFromCurve(FloatCurve=DataInterface() {propertyValues {Curve {keys [<at 0 1>,<at 1 0>]}}}, Scale Curve=0.2)
    )
    Scale RGB=&VectorFromFloat(Value=1)
    ScaleA=true
    ScaleRGB=true
   )
   &DynamicMaterialParameters(
    Index 0 Param 1=1
    Index 0 Param 2=1
    Module.Index0Param0Write=true
    Module.Index0Param1Write=true
    Module.Index0Param2Write=true
    Module.Index0Param3Write=true
   ) {enabled=false}
   /SetVariables(
    Particles.Green WPO:NiagaraPosition=&AddVectorToPosition(
     Position=$Particles.Position
     Vector=&TransformVector(
      Vector=&MakeVector(
       X=0
       Y=&Cosine(
        Normalized Angle=&Add_Float(
         A=&Multiply_Float(A=$Particles.NormalizedAge, B=-1.95)
         B=&Multiply_Float(A=$Emitter.LoopedAge, B=2)
        )
        Period=0.65
        Scale=&FloatFromCurve(FloatCurve=DataInterface() {propertyValues {Curve {keys [<at 0 0>,<at 0.35 1.25>]}}}, Scale Curve=16)
        Bias=0
       )
       Z=&Sine(
        Normalized Angle=&Add_Float(
         A=&Multiply_Float(A=$Particles.NormalizedAge, B=-1.95)
         B=&Multiply_Float(A=$Emitter.LoopedAge, B=2)
        )
        Period=0.65
        Scale=&FloatFromCurve(FloatCurve=DataInterface() {propertyValues {Curve {keys [<at 0 0>,<at 0.35 1.25>]}}}, Scale Curve=16)
        Bias=0
       )
      )
      Destination Space=::Simulation
      Apply Scale=true
     )
    )
    Particles.Yellow WPO:NiagaraPosition=&AddVectorToPosition(
     Position=$Particles.Position
     Vector=&TransformVector(
      Vector=&MakeVector(
       X=0
       Y=&Cosine(
        Normalized Angle=&Add_Float(
         A=&Multiply_Float(A=$Particles.NormalizedAge, B=-1.25)
         B=&Multiply_Float(A=$Emitter.LoopedAge, B=0.9)
        )
        Period=0.55
        Scale=&FloatFromCurve(FloatCurve=DataInterface() {propertyValues {Curve {keys [<at 0 0>,<at 0.4 1.2>]}}}, Scale Curve=16)
        Bias=0
       )
       Z=&Sine(
        Normalized Angle=&Add_Float(
         A=&Multiply_Float(A=$Particles.NormalizedAge, B=-1.25)
         B=&Multiply_Float(A=$Emitter.LoopedAge, B=0.9)
        )
        Period=0.55
        Scale=&FloatFromCurve(FloatCurve=DataInterface() {propertyValues {Curve {keys [<at 0 0>,<at 0.4 1.2>]}}}, Scale Curve=16)
        Bias=0
       )
      )
      Destination Space=::Simulation
      Apply Scale=true
     )
    )
   ) {isSetParameters=true}
   /SetVariables(
    Particles.Ribbon1Dynamic1:Vector4f=&MakeVector4(X=5, Y=0, Z=0, W=0)
    Particles.Ribbon2Dynamic1:Vector4f=&MakeVector4(
     X=1
     Y=2
     Z=&FloatFromCurve(
      FloatCurve=DataInterface() {propertyValues {Curve {keys [{tangentMode=RCTM_User, time=0, value=0}, {tangentMode=RCTM_User, time=0.5, value=1}]}}}
      Scale Curve=13
     )
     W=0
    )
    Particles.RibbonColor1:LinearColor=&VC_DynamicScaleColor(
     Color to Scale:LinearColor=$Particles.Initial.ColorTag1
     RGB Scale:Vector3f=&VectorFromFloat(Value=&MakeCustomFloatFromBool(
       Bool=$User.bIsPainted
       True Float=&FloatFromCurve(
        FloatCurve=DataInterface() {propertyValues {Curve {keys [{tangentMode=RCTM_User, time=0, value=0.3}, {tangentMode=RCTM_User, time=0.5, value=1}]}}}
        Scale Curve=$User.ColorVar1
       )
       False Float=&FloatFromCurve(FloatCurve=DataInterface() {propertyValues {Curve {keys [{tangentMode=RCTM_User, time=0, value=0.5}, {tangentMode=RCTM_User, time=0.5, value=1}]}}})
      ))
     Alpha Scale:NiagaraFloat=&FloatFromCurve(FloatCurve=DataInterface() {propertyValues {Curve {keys [<at 0 0>,<at 0.2 1>,<at 0.5 1>,<at 0.7 0>]}}})
    )
    Particles.RibbonColor2:LinearColor=&VC_DynamicScaleColor(
     Color to Scale:LinearColor=$Particles.Initial.ColorTag2
     RGB Scale:Vector3f=&VectorFromFloat(Value=&MakeCustomFloatFromBool(
       Bool=$User.bIsPainted
       True Float=&FloatFromCurve(FloatCurve=DataInterface() {propertyValues {Curve {keys [{tangentMode=RCTM_User, time=0, value=0.3}, {tangentMode=RCTM_User, time=0.5, value=1}]}}})
       False Float=&FloatFromCurve(FloatCurve=DataInterface() {propertyValues {Curve {keys [{tangentMode=RCTM_User, time=0, value=0.3}, {tangentMode=RCTM_User, time=0.5, value=1}]}}})
      ))
     Alpha Scale:NiagaraFloat=&FloatFromCurve(FloatCurve=DataInterface() {propertyValues {Curve {keys [<at 0 0>,<at 0.2 1>,<at 0.5 1>,<at 0.7 0>]}}})
    )
    Particles.RibbonWidth1:NiagaraFloat=&Multiply_Float(
     A=$Particles.Initial.WidthTag1
     B=&FloatFromCurve(FloatCurve=DataInterface() {propertyValues {Curve {keys [{leaveTangentWeight=0.0333, time=0, value=0}, {arriveTangentWeight=0.0333, time=0.15, value=1}, <at 0.8 1>, <at 1 0>]}}})
    )
    Particles.RibbonWidth2:NiagaraFloat=&Multiply_Float(
     A=$Particles.WidthTag2
     B=&FloatFromCurve(FloatCurve=DataInterface() {propertyValues {Curve {keys [{interpMode=RCIM_Cubic, leaveTangentWeight=0.0333, time=0, value=0}, {arriveTangentWeight=0.0333, interpMode=RCIM_Cubic, time=0.3, value=1}]}}})
    )
   ) {isSetParameters=true}
  ]
  rendererClasses [&NiagaraRibbonRendererProperties]
  renderers [
   {rendererClass=&NiagaraRibbonRendererProperties}
   {rendererClass=&NiagaraRibbonRendererProperties, rendererIndex=1}
   {rendererClass=&NiagaraRibbonRendererProperties, rendererIndex=2}
  ]
 }
 TrailSprites {
  emitterUpdateScript [
   &EmitterState()
   &SpawnRate(SpawnRate=15, Use Spawn Probability=false)
   &SpawnRate(SpawnGroup=1, SpawnRate=15, Use Spawn Probability=false) {module=SpawnRate001}
   &SpawnRate(SpawnGroup=2, SpawnRate=50, Use Spawn Probability=false) {module=SpawnRate002}
  ]
  particleSpawnScript [
   /SetVariables(Particles.RenderTag:NiagaraInt32=&SpawnGroup()) {isSetParameters=true}
   &VC_PaintedVariantsOverride(
    Cobalt=<rgba 0 0.0185 1 1>
    R1 Black:LinearColor=<rgba 0.0417 0.0417 0.0417 1>
    R1 BurntSienna:LinearColor=<rgba 0.0729 0.0205 0.0065 5>
    R1 Cobalt:LinearColor=<rgba 0.4297 0.4495 1.5 1>
    R1 Crimson:LinearColor=<rgba 0.2292 0 0.0062 2>
    R1 DefaultColor=&SelectLinearColorFromArray(
     Array Sampling Mode=::NewEnumerator0
     Direct Array Index=?
     Color Selection Array=DataInterface() {propertyValues {ColorData [<rgba 0.7011 0.7758 0.159 1>, <rgba 0.1683 0.8148 0.0561 1>]}}
     Random Seed=0
    )
    R1 Lime:LinearColor=<rgba 0.6354 0.7652 0 2>
    R1 Pink:LinearColor=<rgba 0.7396 0.3544 0.5995 1>
    R1 Purple:LinearColor=<rgba 0.1988 0 0.5208 2>
    R1 SkyBlue:LinearColor=<rgba 0.2448 0.7473 1 1>
    Use R1 Painted Overrides=true
   )
   &InitializeParticle(
    Color Mode=::NewEnumerator0
    Lifetime=&RandomRangeFloat(Minimum=1, Maximum=1.5)
    Position Offset=&AddVectorToPosition(
     Position=<-100 0 0>
     Vector=&AddVectorToPosition(Position=$User.DTCarOffset, Vector=Linked() {linkedVariable {name=User.SingleBoostPos, type=NiagaraPosition}})
    )
    Position Offset Coordinate Space=::Local
    Sprite Rotation Mode=::NewEnumerator2
    Sprite Size Mode=::NewEnumerator3
    Uniform Sprite Size=&RandomRangeFloat(Minimum=8, Maximum=30)
    UsePositionOffset=true
    Write Material Random=true
   )
   &AddVelocity(
    Cone Angle=/VehicleCosmetics/FX/Niagara/Modules/MergeEmitters/VC_FloatBySpawnGroup(
     SpawnGroup or ExecIndex:NiagaraInt32=0
     CustomInt:NiagaraInt32=0
     MaximumSpawnGroups:NiagaraInt32=3
     Float SG 0:NiagaraFloat=15
     Float SG 1:NiagaraFloat=25
     Float SG 2:NiagaraFloat=0
     Float SG 3:NiagaraFloat=0
     Float SG 4:NiagaraFloat=0
     Float SG 5:NiagaraFloat=0
    )
    Module.UseConeFalloff=false
    Module.UseVelDistribution=false
    Override Seed=false
    Velocity Mode=::NewEnumerator2
    Velocity Speed=&RandomRangeFloat(Minimum=-800, Maximum=-1000)
   )
   &ShapeLocation(MaskSpawnGroup=false, Override Seed=false, Sphere Radius=20, UseOwnerScale=true)
   &Color(Color=&MakeCustomLinearColorFromBool(Bool=$User.bIsPainted, True Color=$User.PaintedColor, False Color=<rgba 1 1 1 1>))
   /SetVariables(
    Particles.ColorTag1:LinearColor=&MakeCustomLinearColorFromBool(Bool=$User.bIsPainted, True Color=$User.PaintedColor, False Color=<rgba 0.6 0.8 0.2 1>)
    Particles.ColorTag2:LinearColor=&MakeCustomLinearColorFromBool(
     Bool=$User.bIsPainted
     True Color=$User.PaintedColor
     False Color=&SelectLinearColorFromArray(
      Array Sampling Mode=::NewEnumerator0
      Direct Array Index=?
      Color Selection Array=DataInterface() {propertyValues {ColorData [<rgba 0.7011 0.7758 0.159 1>, <rgba 0.1683 0.8148 0.0561 1>]}}
      Random Seed=0
     )
    )
    Particles.SizeTag2:Vector2f=&RandomRangeVector2D(Minimum=<1.5 80>, Maximum=<3 140>)
   ) {isSetParameters=true}
  ]
  particleUpdateScript [
   &ParticleState()
   &InheritSourceMovement(
    Applied Position Scale=&VectorFromFloat(Value=&FloatFromCurve(FloatCurve=DataInterface() {propertyValues {Curve {keys [<at 0 1 RCIM_Cubic>,<at 1 0 RCIM_Cubic>]}}}))
    ApplyForce=false
    ApplyPosition=true
    ApplyPositionFalloff=false
   )
   &SpriteRotationRate(Rotation Rate=&RandomRangeFloat(Minimum=-5, Maximum=5))
   &ScaleSpriteSize(Uniform Curve Sprite Scale=DataInterface() {propertyValues {Curve {keys [
        {interpMode=RCIM_Cubic, tangentMode=RCTM_User, time=0, value=1}
        {arriveTangent=-3.1272, interpMode=RCIM_Cubic, leaveTangent=-3.1272, tangentMode=RCTM_User, time=1, value=0}
       ]}}})
   &Drag(
    Drag=&VC_FloatBySpawnGroup(
     SpawnGroup or ExecIndex:NiagaraInt32=0
     CustomInt:NiagaraInt32=0
     MaximumSpawnGroups:NiagaraInt32=3
     Float SG 0:NiagaraFloat=2
     Float SG 1:NiagaraFloat=2
     Float SG 2:NiagaraFloat=0
     Float SG 3:NiagaraFloat=0
     Float SG 4:NiagaraFloat=0
     Float SG 5:NiagaraFloat=0
    )
    Use Linear Drag=true
    Use Rotational Drag=true
   )
   &SolveForcesAndVelocity(Clamp Velocity=false, Limit Acceleration=false)
   &ScaleColor(
    Scale RGB=&MakeCustomVectorFromBool(
     Bool=$User.bIsPainted
     True Vector=&VectorFromFloat(Value=0.5)
     False Vector=&VectorFromCurve(VectorCurve=DataInterface() {propertyValues {XCurve {keys [<at 0 0.5>,<at 1 2>]}, YCurve {keys [<at 0 2>,<at 1 1.5>]}, ZCurve {keys [<at 0 0.1>,<at 1 0.1>]}}})
    )
    ScaleA=true
    ScaleRGB=true
   ) {module=ScaleColor001}
   /SetVariables(
    Particles.SpriteColor1:LinearColor=&VC_DynamicScaleColor(
     Color to Scale:LinearColor=$Particles.Initial.ColorTag1
     RGB Scale:Vector3f=&MakeCustomVectorFromBool(
      Bool=$User.bIsPainted
      True Vector=&VectorFromFloat(Value=&FloatFromCurve(
        FloatCurve=DataInterface() {propertyValues {Curve {keys [{tangentMode=RCTM_User, time=0, value=0.3}, {tangentMode=RCTM_User, time=0.5, value=1}]}}}
        Scale Curve=2
       ))
      False Vector=&VectorFromCurve(VectorCurve=DataInterface() {propertyValues {XCurve {keys [<at 0 0.2>,<at 1 0.1>]}, YCurve {keys [<at 0 2>,<at 1 1.5>]}, ZCurve {keys [<at 0 0.2>,<at 1 0.2>]}}})
     )
     Alpha Scale:NiagaraFloat=1
    )
    Particles.SpriteColor2:LinearColor=&VC_DynamicScaleColor(
     Color to Scale:LinearColor=$Particles.Initial.R1 Color
     RGB Scale:Vector3f=&VectorFromFloat(Value=&MakeCustomFloatFromBool(
       Bool=$User.bIsPainted
       True Float=3
       False Float=&FloatFromCurve(FloatCurve=DataInterface() {propertyValues {Curve {keys [<at 0 1>,<at 1 0>]}}}, Scale Curve=2)
      ))
     Alpha Scale:NiagaraFloat=&FloatFromCurve(FloatCurve=DataInterface() {propertyValues {Curve {keys [<at 0 1>,<at 1 0>]}}})
    )
    Particles.SpriteSize2=&Multiply_Vector2D_ByFloat(
     Float=&FloatFromCurve(FloatCurve=DataInterface() {propertyValues {Curve {keys [{tangentMode=RCTM_User, time=0, value=1}, {tangentMode=RCTM_User, time=0.7, value=0}]}}})
     Vector2D=$Particles.Initial.SizeTag2
    )
   ) {isSetParameters=true}
  ]
  rendererClasses [&NiagaraSpriteRendererProperties]
  renderers [
   {rendererClass=&NiagaraSpriteRendererProperties}
   {rendererClass=&NiagaraSpriteRendererProperties, rendererIndex=1}
   {rendererClass=&NiagaraSpriteRendererProperties, rendererIndex=2}
  ]
 }
}
summary {
 systemName=NS_Boost_ForetstSpirit
 emitters [
  {emitterName=TrailSprites, rendererClasses [&NiagaraSpriteRendererProperties]}
  {emitterName=Ribbon002, rendererClasses [&NiagaraRibbonRendererProperties]}
  {emitterName=Cone, rendererClasses [&NiagaraSpriteRendererProperties, &NiagaraMeshRendererProperties]}
  {emitterName=Buterflies001, rendererClasses [&NiagaraSpriteRendererProperties, &NiagaraMeshRendererProperties]}
 ]
 userVariables [
  User.bIsFrontendPreview {type=NiagaraBool}
  User.bIsPainted {type=NiagaraBool}
  User.ColorVar1 {type=NiagaraFloat, defaultValue=1, description=""}
  User.ColorVariantID {type=NiagaraInt32}
  User.DTCarOffset {type=NiagaraPosition, description="", defaultValue {}}
  User.DTCarOffset001 {type=NiagaraPosition, description="", defaultValue {}}
  User.PaintedColor {type=LinearColor, description="", defaultValue {struct=LinearColor, value=<rgba 1 1 1 1>}}
  User.PaintedColor2 {type=LinearColor, description="", defaultValue {struct=LinearColor, value=<rgba 1 1 1 1>}}
  User.PaintedScalar {type=NiagaraFloat, defaultValue=1, description=""}
  User.ParticleColorVar {type=NiagaraFloat, defaultValue=1, description=""}
  User.PortCount {type=NiagaraInt32, defaultValue=1}
  User.SingleBoostPos {type=NiagaraPosition, description="", defaultValue {}}
 ]
}
